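# Post-process Animation Blueprint for the male medium CadetFrog head that adjusts eye and eyelid bones, apparently to shape blinks.
# The update event remaps the L and R blink pose curves into overlapping, staged weight variables (Initial, Mid, Final, End).
# The AnimGraph chains Modify Bone nodes on left lids and eye, caches that pose, then applies the right-side chain.
# No visible link connects the weights to node alphas.
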AnimBlueprint /BRCosmetics/Characters/Player/Male/Medium/Heads/M_MED_CadetFrog_Head/Meshes/CustomSkeleton/M_MED_Head_CadetFrog_PostProcess_AnimB
target_skeleton=&/Fortnite_Base_Head_CadetFrog_Skeleton
graphs [
 AnimGraph {nodes [
   n1 = "Transform (Modify) Bone - Bone: R_eye_double_lid_lower_mid"!(ComponentPose=Local To Component(LocalPose=Use cached pose 'L_Blink_Pose'()), Alpha=1, Rotation="0,0,10", Scale="1,1,1") -> (Pose) {
    rotationMode=BMM_Additive
    rotationSpace=BCS_BoneSpace
    scaleSpace=BCS_BoneSpace
    translationSpace=BCS_BoneSpace
    alphaBoolBlend {blendOption=Linear}
    alphaScaleBiasClamp {bInterpResult=true, interpSpeedDecreasing=6, interpSpeedIncreasing=16}
    boneToModify {boneName=R_eye_double_lid_lower_mid}
    rotation {roll=10}
   }
   n2 = "Transform (Modify) Bone - Bone: R_eye_double_lid_lower_mid_bnd"!(ComponentPose=@n1, Alpha=1, Rotation="0,0,0", Scale="1,1,2") -> (Pose) {
    rotationMode=BMM_Additive
    rotationSpace=BCS_BoneSpace
    scaleMode=BMM_Additive
    scaleSpace=BCS_BoneSpace
    translationSpace=BCS_BoneSpace
    alphaBoolBlend {blendOption=Linear}
    alphaScaleBiasClamp {bInterpResult=true, interpSpeedDecreasing=5, interpSpeedIncreasing=16}
    boneToModify {boneName=R_eye_double_lid_lower_mid_bnd}
    scale {z=2}
   }
   n3 = "Transform (Modify) Bone - Bone: R_eye_double_lid_lower_mid"(ComponentPose=@n2, Alpha=1, Rotation="0,0,30", Scale="1,1,1") {
    rotationMode=BMM_Additive
    rotationSpace=BCS_BoneSpace
    scaleSpace=BCS_BoneSpace
    translationSpace=BCS_BoneSpace
    alphaBoolBlend {blendOption=Linear}
    alphaScaleBiasClamp {bInterpResult=true, interpSpeedDecreasing=6, interpSpeedIncreasing=16}
    boneToModify {boneName=R_eye_double_lid_lower_mid}
    rotation {roll=30}
   }
   n4 = "Transform (Modify) Bone - Bone: R_eye_lid_upper_mid"!(ComponentPose=@n3, Alpha=1, Translation="-0.6,0,-0.6", Rotation="0,0,0") -> (Pose) {
    rotationMode=BMM_Additive
    rotationSpace=BCS_BoneSpace
    scaleMode=BMM_Additive
    scaleSpace=BCS_BoneSpace
    translationMode=BMM_Additive
    translationSpace=BCS_BoneSpace
    alphaBoolBlend {blendOption=Linear}
    alphaScaleBiasClamp {bInterpResult=true, interpSpeedDecreasing=0, interpSpeedIncreasing=0}
    boneToModify {boneName=R_eye_lid_upper_mid}
    scale {x=1.05}
    translation {x=-0.6, z=-0.6}
   }
   n5 = "Transform (Modify) Bone - Bone: R_eye"!(ComponentPose=@n4, Alpha=1, Translation="0,0.2,-1", Rotation="0,0,0") -> (Pose) {
    rotationMode=BMM_Additive
    rotationSpace=BCS_BoneSpace
    scaleMode=BMM_Additive
    scaleSpace=BCS_BoneSpace
    translationMode=BMM_Additive
    translationSpace=BCS_BoneSpace
    alphaBoolBlend {blendOption=Linear}
    alphaScaleBiasClamp {bInterpResult=true, interpSpeedDecreasing=0, interpSpeedIncreasing=0}
    boneToModify {boneName=R_eye}
    scale {x=1.05}
    translation {y=0.2, z=-1}
   }
   Output Pose(Result=Component To Local(ComponentPose=@n5))
   n6 = "Transform (Modify) Bone - Bone: L_eye_double_lid_lower_mid"!(ComponentPose=Local To Component(LocalPose=Input Pose()), Alpha=1, Rotation="0,0,10", Scale="1,1,1") -> (Pose) {
    rotationMode=BMM_Additive
    rotationSpace=BCS_BoneSpace
    scaleSpace=BCS_BoneSpace
    translationSpace=BCS_BoneSpace
    alphaBoolBlend {blendOption=Linear}
    alphaScaleBiasClamp {bInterpResult=true, interpSpeedDecreasing=6, interpSpeedIncreasing=16}
    boneToModify {boneName=L_eye_double_lid_lower_mid}
    rotation {roll=10}
    translation {z=1}
   }
   n7 = "Transform (Modify) Bone - Bone: L_eye_double_lid_lower_mid_bnd"!(ComponentPose=@n6, Alpha=1, Rotation="0,0,0", Scale="1,1,2") -> (Pose) {
    rotationMode=BMM_Additive
    rotationSpace=BCS_BoneSpace
    scaleMode=BMM_Additive
    scaleSpace=BCS_BoneSpace
    translationSpace=BCS_BoneSpace
    alphaBoolBlend {blendOption=Linear}
    alphaScaleBiasClamp {bInterpResult=true, interpSpeedDecreasing=5, interpSpeedIncreasing=16}
    boneToModify {boneName=L_eye_double_lid_lower_mid_bnd}
    scale {z=2}
    translation {z=1}
   }
   n8 = "Transform (Modify) Bone - Bone: L_eye_double_lid_lower_mid"(ComponentPose=@n7, Alpha=1, Rotation="0.000,0.000,27.947", Scale="1,1,1") {
    rotationMode=BMM_Additive
    rotationSpace=BCS_BoneSpace
    scaleSpace=BCS_BoneSpace
    translationSpace=BCS_BoneSpace
    alphaBoolBlend {blendOption=Linear}
    alphaScaleBiasClamp {bInterpResult=true, interpSpeedDecreasing=6, interpSpeedIncreasing=16}
    boneToModify {boneName=L_eye_double_lid_lower_mid}
    rotation {roll=27.947}
    translation {z=1}
   }
   n9 = "Transform (Modify) Bone - Bone: L_eye_lid_upper_mid"!(ComponentPose=@n8, Alpha=1, Translation="0.5,0,-0.45", Rotation="0,0,0") -> (Pose) {
    rotationMode=BMM_Additive
    rotationSpace=BCS_BoneSpace
    scaleMode=BMM_Additive
    scaleSpace=BCS_BoneSpace
    translationMode=BMM_Additive
    translationSpace=BCS_BoneSpace
    alphaBoolBlend {blendOption=Linear}
    alphaScaleBiasClamp {bInterpResult=true, interpSpeedDecreasing=0, interpSpeedIncreasing=0}
    boneToModify {boneName=L_eye_lid_upper_mid}
    scale {x=1.05}
    translation {x=0.5, z=-0.45}
   }
   n10 = "Transform (Modify) Bone - Bone: L_eye"!(ComponentPose=@n9, Alpha=1, Translation="0,0.2,-1", Rotation="0,0,0") -> (Pose) {
    rotationMode=BMM_Additive
    rotationSpace=BCS_BoneSpace
    scaleMode=BMM_Additive
    scaleSpace=BCS_BoneSpace
    translationMode=BMM_Additive
    translationSpace=BCS_BoneSpace
    alphaBoolBlend {blendOption=Linear}
    alphaScaleBiasClamp {bInterpResult=true, interpSpeedDecreasing=0, interpSpeedIncreasing=0}
    boneToModify {boneName=L_eye}
    scale {x=1.05}
    translation {y=0.2, z=-1}
   }
   Save cached pose 'L_Blink_Pose'(Pose=Component To Local(ComponentPose=@n10)) {cacheName=L_Blink_Pose}
  ]}
 EventGraph {nodes [
   Event BlueprintUpdateAnimation()
   > n1 = Sequence!(execute) -> (then_0, then_1, then_2:Exec)
   n2 = MapRangeClamped(Value=GetCurveValue(CurveName=L_blink_pose), InRangeA=0, InRangeB=2)
   n3 = Remap Float(Input=@n2, OldRange=(X=0,Y=0.25), NewRange=(X=0,Y=1), Clamp=true) -> (Output=~)
   > Set L_blinkPose_Initial(L_blinkPose_Initial:double=float - float(A=@n3, B=Remap Float(Input=@n2, OldRange=(X=0.5,Y=1), NewRange=(X=0,Y=1), Clamp=true) -> (Output=~))) -> (Output_Get:double)
   > Set L_blinkPose_Mid(L_blinkPose_Mid:double=@n3) -> (Output_Get:double)
   > Set L_blinkPose_Final(L_blinkPose_Final:double=Remap Float(Input=@n2, OldRange=(X=0.5,Y=1), NewRange=(X=0,Y=1), Clamp=true) -> (Output=~)) -> (Output_Get:double)
   > Set L_blinkPose(L_blinkPose:float=@n2) -> (Output_Get:float)
   > Set L_blinkPose_End(L_blinkPose_End:double=Remap Float(Input=@n2, OldRange=(X=0.99,Y=1), NewRange=(X=0,Y=1), Clamp=true) -> (Output=~)) -> (Output_Get:double)
   n4 = MapRangeClamped(Value=GetCurveValue(CurveName=R_blink_pose), InRangeA=0, InRangeB=2)
   n5 = Remap Float(Input=@n4, OldRange=(X=0.5,Y=1), NewRange=(X=0,Y=1), Clamp=true) -> (Output=~)
   Set R_blinkPose_Initial(
    execute=@n1.then_1
    R_blinkPose_Initial:double=float - float(A=Remap Float(Input=@n4, OldRange=(X=0,Y=0.25), NewRange=(X=0,Y=1), Clamp=true) -> (Output=~), B=@n5)
   ) -> (Output_Get:double)
   > Set R_blinkPose_Mid(R_blinkPose_Mid:double=Remap Float(Input=@n4, OldRange=(X=0,Y=0.25), NewRange=(X=0,Y=1), Clamp=true) -> (Output=~)) -> (Output_Get:double)
   > Set R_blinkPose_Final(R_blinkPose_Final:double=@n5) -> (Output_Get:double)
   > Set R_blinkPose(R_blinkPose:float=@n4) -> (Output_Get:float)
   > Set R_blinkPose_End(R_blinkPose_End:double=Remap Float(Input=@n4, OldRange=(X=0.99,Y=1), NewRange=(X=0,Y=1), Clamp=true) -> (Output=~)) -> (Output_Get:double)
   TryGetPawnOwner()
  ]}
]
variables [
 L_blinkPose {type=number}
 L_blinkPose_End {type=number}
 L_blinkPose_Final {type=number}
 L_blinkPose_Initial {type=number}
 L_blinkPose_Mid {type=number}
 R_blinkPose {type=number}
 R_blinkPose_End {type=number}
 R_blinkPose_Final {type=number}
 R_blinkPose_Initial {type=number}
 R_blinkPose_Mid {type=number}
]
